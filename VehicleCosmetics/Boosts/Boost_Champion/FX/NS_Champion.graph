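# Niagara system for the Champion vehicle boost cosmetic, a purple trail of ribbons, sprites and stars emitted from the vehicle's boost sockets.
# The ribbon emitter feeds six renderers, computing three sine/cosine-offset position tracks for weaving strands.
# Sprite and Star emitters partition particles across sockets, adding soft-fading sprites and backward cone-sprayed SubUV stars.
# Game-mode masks and a frontend-preview flag retune rates, sizes and velocities; painted-variant overrides are disabled.
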NiagaraSystem /VehicleCosmetics/Boosts/Boost_Champion/FX/NS_Champion
dependencies {
 usedDataInterfaces [
  &NiagaraDataInterfaceCurve
  &NiagaraDataInterfaceSpriteRendererInfo
  &NiagaraDataInterfaceArrayFloat
  &NiagaraDataInterfaceColorCurve
  &NiagaraDataInterfaceSkeletalMesh
  &NiagaraDataInterfaceArrayFloat3
 ]
 usedDynamicInputs [&SelectVectorFromArray]
 usedModules [
  &self:SystemSpawnScript.SystemScriptSource.SystemScriptGraph.NiagaraNodeAssignment_0.SetVariables_#1
  &self:SystemSpawnScript.SystemScriptSource.SystemScriptGraph.NiagaraNodeAssignment_1.SetVariables_#2
  &self:SystemSpawnScript.SystemScriptSource.SystemScriptGraph.NiagaraNodeAssignment_2.SetVariables_#3
  &SystemState
 ]
}
emitters {
 NE_Champion_Ribbon {
  emitterUpdateScript [
   &EmitterState(Life Cycle Mode=::NewEnumerator1, Loop Duration=1, UseLoopDelay=false)
   &VC_FilteredSocketCount()
   &VC_RateQualityHandler(Default Rate:NiagaraFloat=&Multiply_Float(
     A=100
     B=/VehicleCosmetics/FX/Niagara/Modules/VC_Mask_Floats(Coro Value:NiagaraFloat=1, FrontEnd Value:NiagaraFloat=0.5, FN Value:NiagaraFloat=0.5)
    ))
   &VC_SpawnRibbon_v1(Number of Ribbons=$Emitter.PortCount, Spawn Time=-1, SpawnRate=$Emitter.Spawn Rate)
  ]
  particleSpawnScript [
   &VC_PaintedVariantsOverride() {enabled=false}
   &VC_InitializeRibbon_v1(
    Custom Offset=true
    Enable Skeletal Position=true
    Lifetime=&MakeCustomFloatFromBool(Bool=$User.bIsFrontendPreview, True Float=0.5, False Float=2)
    Local Vector Offset=<60 0 0>
    Offset Amount:NiagaraFloat=&MakeCustomFloatFromBool(Bool=$User.bIsFrontendPreview, True Float=1, False Float=0)
    Ribbon Width:NiagaraFloat=&MakeCustomFloatFromBool(Bool=$User.bIsFrontendPreview, True Float=6, False Float=13)
   )
   &AddVelocity(Velocity=&VC_Mask_Vectors(
     Coro Value:Vector3f=<-200 0 0>
     FrontEnd Value:Vector3f=&MakeCustomVectorFromBool(Bool=$User.bIsFrontendPreview, True Vector=<-3000 0 1>, False Vector=<-400 0 0>)
     FN Value:Vector3f=<-1200 0 0>
    )) {module=AddVelocity001}
   /VehicleCosmetics/FX/VC_FXParents/Modules/DLR_ProductPreviews(GameModeID:NiagaraInt32=$System.GameModeID, bPreviewVehicleSpeed:NiagaraBool=$System.bPreviewVehicleSpeed)
  ]
  particleUpdateScript [
   &ParticleState()
   &InheritSourceMovement(
    Applied Position Scale=&VectorFromFloat(Value=&FloatFromCurve(FloatCurve=DataInterface() {propertyValues {Curve {
         defaultValue=1
         keys [
          {arriveTangent=-1.7885, interpMode=RCIM_Cubic, leaveTangent=-1.7885, tangentMode=RCTM_User, time=0.05, value=1}
          {interpMode=RCIM_Cubic, tangentMode=RCTM_User, time=1, value=0}
         ]
        }}}))
    ApplyForce=false
    ApplyPosition=true
    ApplyPositionFalloff=false
   )
   &SolveForcesAndVelocity(Clamp Velocity=false, Limit Acceleration=false)
   /SetVariables(Particles.RibbonColor:LinearColor=&MultiplyLinearColor(
     Linear Color=&MakeCustomLinearColorFromBool(Bool=$User.bIsFrontendPreview, True Color=<rgba 0.4207 0 1.2 1>, False Color=<rgba 0.4485 0.0776 1.2 1>)
     Scale Factor=&MakeVector4FromVectorAndFloat("Vector (RGB)"=<1 1 1>, "Float (Alpha)"=1)
    )) {isSetParameters=true}
   /SetVariables(Particles.RibbonWidth02:NiagaraFloat=&Multiply_Float(
     A=&FloatFromCurve(FloatCurve=DataInterface() {propertyValues {Curve {keys [<at 0 1.2 RCIM_Cubic>,<at 1 0 RCIM_Cubic>]}}})
     B=&MakeCustomFloatFromBool(Bool=$User.bIsFrontendPreview, True Float=80, False Float=130)
    )) {isSetParameters=true}
   &VC_UpdateRibbon_v1(Enable Skeletal Position=true)
   /SetVariables(Particles.WaveScaleCurve:NiagaraFloat=&MakeCustomFloatFromBool(
     Bool=$User.bIsFrontendPreview
     True Float=&FloatFromCurve(FloatCurve=DataInterface() {propertyValues {Curve {keys [
          {
           arriveTangent=0.9102
           arriveTangentWeight=0.2566
           interpMode=RCIM_Cubic
           leaveTangent=0.9102
           leaveTangentWeight=0.2566
           tangentMode=RCTM_User
           tangentWeightMode=RCTWM_WeightedBoth
           time=0
           value=0.3
          }
          {arriveTangent=0.5064, interpMode=RCIM_Cubic, leaveTangent=0.5064, tangentMode=RCTM_User, time=0.4, value=0.65}
          {arriveTangent=0.0096, interpMode=RCIM_Cubic, leaveTangent=0.0096, tangentMode=RCTM_User, time=1, value=0.8}
         ]}}})
     False Float=&FloatFromCurve(FloatCurve=DataInterface() {propertyValues {Curve {keys [
          {
           arriveTangent=0.9102
           arriveTangentWeight=0.2566
           interpMode=RCIM_Cubic
           leaveTangent=0.9102
           leaveTangentWeight=0.2566
           tangentMode=RCTM_User
           tangentWeightMode=RCTWM_WeightedBoth
           time=0
           value=0.3
          }
          {arriveTangent=0.5064, interpMode=RCIM_Cubic, leaveTangent=0.5064, tangentMode=RCTM_User, time=0.4, value=0.65}
          {arriveTangent=0.0096, interpMode=RCIM_Cubic, leaveTangent=0.0096, tangentMode=RCTM_User, time=1, value=0.8}
         ]}}})
    )) {isSetParameters=true}
   /SetVariables(
    Particles.CustomPosition_R1:NiagaraPosition=&AddVectorToPosition(
     Position=$Particles.Position
     Vector=&Multiply_VectorByFloat(
      Vector=&MakeVector(
       X=&Sine(
        Normalized Angle=&Add_Float(A=0, B=&Subtract_Float(A=$Emitter.NormalizedLoopAge, B=$Particles.NormalizedAge))
        Period=&MakeCustomFloatFromBool(Bool=$User.bIsFrontendPreview, True Float=0.25, False Float=0.35)
        Scale=&Multiply_Float(
         A=&SelectFloatFromArray(Direct Array Index=$Particles.SpawnGroup, Float Selection Array=DataInterface() {propertyValues {FloatData [1.5,1.5]}})
         B=&MakeCustomFloatFromBool(Bool=$User.bIsFrontendPreview, True Float=3, False Float=12)
        )
        Bias=0
       )
       Y=&Sine(
        Normalized Angle=&Add_Float(A=0, B=&Subtract_Float(A=$Emitter.NormalizedLoopAge, B=$Particles.NormalizedAge))
        Period=&MakeCustomFloatFromBool(Bool=$User.bIsFrontendPreview, True Float=0.5, False Float=0.35)
        Scale=&Multiply_Float(
         A=&SelectFloatFromArray(Direct Array Index=$Particles.SpawnGroup, Float Selection Array=DataInterface() {propertyValues {FloatData [2.5,-2.5]}})
         B=&MakeCustomFloatFromBool(Bool=$User.bIsFrontendPreview, True Float=5, False Float=12)
        )
        Bias=&MakeCustomFloatFromBool(Bool=$User.bIsFrontendPreview, True Float=80, False Float=30)
       )
       Z=&Cosine(
        Normalized Angle=&Add_Float(A=1, B=&Subtract_Float(A=$Emitter.NormalizedLoopAge, B=$Particles.NormalizedAge))
        Period=&MakeCustomFloatFromBool(Bool=$User.bIsFrontendPreview, True Float=0.25, False Float=0.35)
        Scale=&Multiply_Float(
         A=&SelectFloatFromArray(Direct Array Index=$Particles.SpawnGroup, Float Selection Array=DataInterface() {propertyValues {FloatData [5,5]}})
         B=&MakeCustomFloatFromBool(Bool=$User.bIsFrontendPreview, True Float=3, False Float=12)
        )
        Bias=0
       )
      )
      Float=$Particles.WaveScaleCurve
     )
    )
    Particles.CustomPosition_R2:NiagaraPosition=&AddVectorToPosition(
     Position=$Particles.Position
     Vector=&Multiply_VectorByFloat(
      Vector=&MakeVector(
       X=&Sine(
        Normalized Angle=&Add_Float(
         A=&Multiply_Float(A=1, B=0)
         B=&Subtract_Float(A=$Emitter.NormalizedLoopAge, B=$Particles.NormalizedAge)
        )
        Period=&MakeCustomFloatFromBool(Bool=$User.bIsFrontendPreview, True Float=0.25, False Float=0.35)
        Scale=&Multiply_Float(
         A=&SelectFloatFromArray(Direct Array Index=$Particles.SpawnGroup, Float Selection Array=DataInterface() {propertyValues {FloatData [1.5,1.5]}})
         B=&MakeCustomFloatFromBool(Bool=$User.bIsFrontendPreview, True Float=3, False Float=12)
        )
        Bias=0
       )
       Y=&Sine(
        Normalized Angle=&Add_Float(
         A=&Multiply_Float(A=1, B=0.1)
         B=&Subtract_Float(A=$Emitter.NormalizedLoopAge, B=$Particles.NormalizedAge)
        )
        Period=&MakeCustomFloatFromBool(Bool=$User.bIsFrontendPreview, True Float=0.25, False Float=0.35)
        Scale=&Multiply_Float(
         A=&SelectFloatFromArray(Direct Array Index=$Particles.SpawnGroup, Float Selection Array=DataInterface() {propertyValues {FloatData [2.5,-2.5]}})
         B=&MakeCustomFloatFromBool(Bool=$User.bIsFrontendPreview, True Float=5, False Float=12)
        )
        Bias=&MakeCustomFloatFromBool(Bool=$User.bIsFrontendPreview, True Float=-80, False Float=-30)
       )
       Z=&Cosine(
        Normalized Angle=&Add_Float(A=0.2, B=&Subtract_Float(A=$Emitter.NormalizedLoopAge, B=$Particles.NormalizedAge))
        Period=&MakeCustomFloatFromBool(Bool=$User.bIsFrontendPreview, True Float=0.25, False Float=0.35)
        Scale=&Multiply_Float(
         A=&MakeCustomFloatFromBool(
          Bool=$User.bIsFrontendPreview
          True Float=&SelectFloatFromArray(Direct Array Index=$Particles.SpawnGroup, Float Selection Array=DataInterface() {propertyValues {FloatData [5,5]}})
          False Float=&SelectFloatFromArray(Direct Array Index=$Particles.SpawnGroup, Float Selection Array=DataInterface() {propertyValues {FloatData [5,5]}})
         )
         B=&MakeCustomFloatFromBool(Bool=$User.bIsFrontendPreview, True Float=3, False Float=12)
        )
        Bias=0
       )
      )
      Float=$Particles.WaveScaleCurve
     )
    )
   ) {isSetParameters=true}
   /SetVariables(Particles.CustomPosition_R3:NiagaraPosition=&AddVectorToPosition(
     Position=$Particles.Position
     Vector=&Multiply_VectorByFloat(
      Vector=&MakeVector(
       X=&Sine(
        Normalized Angle=&Add_Float(
         A=&Multiply_Float(A=1, B=0)
         B=&Subtract_Float(A=$Emitter.NormalizedLoopAge, B=$Particles.NormalizedAge)
        )
        Period=&MakeCustomFloatFromBool(Bool=$User.bIsFrontendPreview, True Float=0.25, False Float=0.35)
        Scale=&Multiply_Float(
         A=&SelectFloatFromArray(Direct Array Index=$Particles.SpawnGroup, Float Selection Array=DataInterface() {propertyValues {FloatData [1.5,1.5]}})
         B=&MakeCustomFloatFromBool(Bool=$User.bIsFrontendPreview, True Float=3, False Float=12)
        )
        Bias=0
       )
       Y=&Sine(
        Normalized Angle=&Add_Float(
         A=&Multiply_Float(A=1, B=0.1)
         B=&Subtract_Float(A=$Emitter.NormalizedLoopAge, B=$Particles.NormalizedAge)
        )
        Period=&MakeCustomFloatFromBool(Bool=$User.bIsFrontendPreview, True Float=0.45, False Float=0.35)
        Scale=&Multiply_Float(
         A=&SelectFloatFromArray(Direct Array Index=$Particles.SpawnGroup, Float Selection Array=DataInterface() {propertyValues {FloatData [2.5,-2.5]}})
         B=&MakeCustomFloatFromBool(Bool=$User.bIsFrontendPreview, True Float=5, False Float=12)
        )
        Bias=0
       )
       Z=&Cosine(
        Normalized Angle=&Add_Float(A=0.2, B=&Subtract_Float(A=$Emitter.NormalizedLoopAge, B=$Particles.NormalizedAge))
        Period=&MakeCustomFloatFromBool(Bool=$User.bIsFrontendPreview, True Float=0.25, False Float=0.35)
        Scale=&Multiply_Float(
         A=&SelectFloatFromArray(Direct Array Index=$Particles.SpawnGroup, Float Selection Array=DataInterface() {propertyValues {FloatData [5,5]}})
         B=&MakeCustomFloatFromBool(Bool=$User.bIsFrontendPreview, True Float=3, False Float=5)
        )
        Bias=0
       )
      )
      Float=$Particles.WaveScaleCurve
     )
    )) {isSetParameters=true}
   &ScaleColor(
    Scale RGB=&MakeVectorFromLinearColorRGB(Color=&MakeCustomLinearColorFromBool(
      Bool=$User.bIsFrontendPreview
      True Color=&ColorFromCurve(ColorCurve=DataInterface() {propertyValues {
         AlphaCurve {keys [{interpMode=RCIM_Cubic, tangentMode=RCTM_User, time=0, value=1}, <at 1 1>]}
         BlueCurve {keys [<at 0 5>, {arriveTangent=-2.023, interpMode=RCIM_Cubic, leaveTangent=-2.023, tangentMode=RCTM_User, time=0.409, value=2.5}]}
         GreenCurve {keys [<at 0 0.7653>, {interpMode=RCIM_Cubic, tangentMode=RCTM_User, time=0.409, value=0}]}
         RedCurve {keys [<at 0 3.8959>, {interpMode=RCIM_Cubic, tangentMode=RCTM_User, time=0.409, value=0.4403}]}
        }})
      False Color=&ColorFromCurve(ColorCurve=DataInterface() {propertyValues {
         AlphaCurve {keys [{interpMode=RCIM_Cubic, tangentMode=RCTM_User, time=0, value=1}, <at 1 1>]}
         BlueCurve {keys [
           {interpMode=RCIM_Cubic, tangentMode=RCTM_User, time=0, value=20}
           {arriveTangent=-2.023, interpMode=RCIM_Cubic, leaveTangent=-2.023, tangentMode=RCTM_User, time=0.2852, value=2.5}
          ]}
         GreenCurve {keys [{interpMode=RCIM_Cubic, tangentMode=RCTM_User, time=0, value=3.0611}, {interpMode=RCIM_Cubic, tangentMode=RCTM_User, time=0.2852, value=0}]}
         RedCurve {keys [
           {interpMode=RCIM_Cubic, tangentMode=RCTM_User, time=0, value=15.5835}
           {interpMode=RCIM_Cubic, tangentMode=RCTM_User, time=0.2852, value=0.7703}
          ]}
        }})
     ))
    ScaleA=true
    ScaleRGB=true
   ) {module=ScaleColor001}
   &ScaleRibbonWidth(Ribbon Width Scale=&MakeCustomFloatFromBool(
     Bool=$User.bIsFrontendPreview
     True Float=&FloatFromCurve(FloatCurve=DataInterface() {propertyValues {Curve {keys [
          {arriveTangent=-0.0602, interpMode=RCIM_Cubic, leaveTangent=-0.0602, tangentMode=RCTM_User, time=0, value=1.5}
          {arriveTangent=-1.5, interpMode=RCIM_Cubic, leaveTangent=-1.5, time=0.3, value=1.375}
          {arriveTangent=0.0531, interpMode=RCIM_Cubic, leaveTangent=0.0531, tangentMode=RCTM_User, time=1, value=0}
         ]}}})
     False Float=&FloatFromCurve(FloatCurve=DataInterface() {propertyValues {Curve {keys [<at 0 1.15 RCIM_Cubic>, {arriveTangent=-1.15, interpMode=RCIM_Cubic, leaveTangent=-1.15, time=0.3, value=1.05}, <at 1 0 RCIM_Cubic>]}}})
    )) {module=ScaleRibbonWidth001}
   &DynamicMaterialParameters(
    Index 0 Param 1=1
    Index 0 Param 4=1
    Module.Index0Float0Write=true
    Module.Index0Float1Write=true
    Module.Index0Float2Write=true
    Module.Index0Float3Write=true
   )
  ]
  rendererClasses [&NiagaraRibbonRendererProperties]
  renderers [
   {rendererClass=&NiagaraRibbonRendererProperties}
   {rendererClass=&NiagaraRibbonRendererProperties, rendererIndex=1}
   {rendererClass=&NiagaraRibbonRendererProperties, rendererIndex=2}
   {rendererClass=&NiagaraRibbonRendererProperties, rendererIndex=3}
   {rendererClass=&NiagaraRibbonRendererProperties, rendererIndex=4}
   {rendererClass=&NiagaraRibbonRendererProperties, rendererIndex=5}
  ]
 }
 NE_Champion_Sprite {
  emitterUpdateScript [
   &EmitterState()
   &VC_FilteredSocketCount()
   &SpawnRate(
    SpawnRate=&Multiply_Float(
     A=&MakeCustomFloatFromBool(Bool=$User.bIsFrontendPreview, True Float=250, False Float=200)
     B=&VC_Mask_Floats(Coro Value:NiagaraFloat=1, FrontEnd Value:NiagaraFloat=1, FN Value:NiagaraFloat=0.5)
    )
    Use Spawn Probability=false
   )
  ]
  particleSpawnScript [
   &VC_PaintedVariantsOverride() {enabled=false}
   &InitializeParticle(
    Color=&MakeCustomLinearColorFromBool(Bool=$User.bIsFrontendPreview, True Color=<rgba 28.2761 2.8646 50 1>, False Color=<rgba 0.5655 0.0573 1 1>)
    Lifetime=&MakeCustomFloatFromBool(
     Bool=$User.bIsFrontendPreview
     True Float=&RandomRangeFloat(Minimum=0.3, Maximum=0.5)
     False Float=&RandomRangeFloat(Minimum=0.3, Maximum=0.5)
    )
    Sprite Rotation Mode=::NewEnumerator2
    Sprite Size Mode=::NewEnumerator3
    Uniform Sprite Size=&MakeCustomFloatFromBool(
     Bool=$User.bIsFrontendPreview
     True Float=&RandomRangeFloat(Minimum=3, Maximum=6)
     False Float=&RandomRangeFloat(Minimum=30, Maximum=70)
    )
    UsePositionOffset=false
    Write Material Random=true
   )
   &PartitionParticles(Number Of Partitions=$Emitter.PortCount)
   &SkeletalMeshLocation(
    Bone / Socket Index=$Particles.PartitionParticles.Partition
    Bone / Socket Sampling Mode=::NewEnumerator5
    MaskSpawnGroup=false
    Mesh Sampling Type=::NewEnumerator6
    Skeletal Mesh=$System.SkeletalMesh
   )
   &AddVelocity(Velocity=&VC_Mask_Vectors(
     Coro Value:Vector3f=<-400 0 0>
     FrontEnd Value:Vector3f=&MakeCustomVectorFromBool(Bool=$User.bIsFrontendPreview, True Vector=<-1500 0 1>, False Vector=<-200 0 0>)
     FN Value:Vector3f=<-1200 0 0>
    ))
   &ShapeLocation(
    MaskSpawnGroup=false
    Override Seed=false
    Sphere Radius=&MakeCustomFloatFromBool(Bool=$User.bIsFrontendPreview, True Float=30, False Float=60)
    UseOwnerScale=true
   )
   &DLR_ProductPreviews(GameModeID:NiagaraInt32=$System.GameModeID, bPreviewVehicleSpeed:NiagaraBool=$System.bPreviewVehicleSpeed)
  ]
  particleUpdateScript [
   &ParticleState()
   &SpriteRotationRate(Rotation Rate=&RandomRangeFloat(Minimum=30, Maximum=-30))
   &ScaleSpriteSize(Uniform Curve Sprite Scale=DataInterface() {propertyValues {Curve {
       defaultValue=0
       keys [
        {arriveTangent=0.9392, interpMode=RCIM_Cubic, leaveTangent=0.9392, tangentMode=RCTM_User, time=0, value=0.8}
        {interpMode=RCIM_Cubic, tangentMode=RCTM_User, time=1, value=1.3}
       ]
      }}})
   &InheritSourceMovement(
    Applied Position Scale=&VectorFromFloat(Value=&FloatFromCurve(FloatCurve=DataInterface() {propertyValues {Curve {
         defaultValue=1
         keys [
          {arriveTangent=-1.7885, interpMode=RCIM_Cubic, leaveTangent=-1.7885, tangentMode=RCTM_User, time=0.05, value=1}
          {interpMode=RCIM_Cubic, tangentMode=RCTM_User, time=1, value=0}
         ]
        }}}))
    ApplyForce=false
    ApplyPosition=true
    ApplyPositionFalloff=false
   )
   &ScaleColor(
    Scale Alpha=&FloatFromCurve(FloatCurve=DataInterface() {propertyValues {Curve {
        defaultValue=0
        keys [
         {interpMode=RCIM_Cubic, tangentMode=RCTM_User, time=0, value=0}
         {arriveTangent=0.0961, interpMode=RCIM_Cubic, leaveTangent=0.0961, tangentMode=RCTM_User, time=0.1, value=1}
         {interpMode=RCIM_Cubic, tangentMode=RCTM_User, time=0.8, value=1}
         {tangentMode=RCTM_User, time=1, value=0}
        ]
       }}})
    Scale RGB=&Multiply_VectorByFloat(Vector=<1 1 1>, Float=1)
    ScaleA=true
    ScaleRGB=true
   ) {module=ScaleColor001}
   &SolveForcesAndVelocity(Clamp Velocity=false, Limit Acceleration=false)
   &DynamicMaterialParameters(
    Index 0 Param 1=1
    Index 0 Param 2=1
    Module.Index0Float0Write=true
    Module.Index0Float1Write=true
    Module.Index0Float2Write=true
    Module.Index0Float3Write=true
   )
  ]
  rendererClasses [&NiagaraSpriteRendererProperties]
  renderers [{rendererClass=&NiagaraSpriteRendererProperties}]
 }
 NE_Champion_Star {
  emitterUpdateScript [
   &EmitterState()
   &VC_FilteredSocketCount()
   &SpawnRate(
    SpawnRate=&Multiply_Float(
     A=&MakeCustomFloatFromBool(Bool=$User.bIsFrontendPreview, True Float=150, False Float=200)
     B=&VC_Mask_Floats(Coro Value:NiagaraFloat=1, FrontEnd Value:NiagaraFloat=0.5, FN Value:NiagaraFloat=0.5)
    )
    Use Spawn Probability=false
   )
  ]
  particleSpawnScript [
   &VC_PaintedVariantsOverride() {enabled=false}
   &InitializeParticle(
    Color=&MakeCustomLinearColorFromBool(Bool=$User.bIsFrontendPreview, True Color=<rgba 0.4672 0.0573 1 1>, False Color=<rgba 0.7007 0.0859 1.5 1>)
    Lifetime=&MakeCustomFloatFromBool(
     Bool=$User.bIsFrontendPreview
     True Float=&RandomRangeFloat(Minimum=0.4, Maximum=0.8)
     False Float=&RandomRangeFloat(Minimum=0.5, Maximum=1)
    )
    Sprite Rotation Mode=::NewEnumerator2
    Sprite Size Mode=::NewEnumerator3
    Uniform Sprite Size=&MakeCustomFloatFromBool(
     Bool=$User.bIsFrontendPreview
     True Float=&RandomRangeFloat(Minimum=15, Maximum=30)
     False Float=&RandomRangeFloat(Minimum=40, Maximum=85)
    )
    UsePositionOffset=false
    Write Material Random=true
   )
   &PartitionParticles(Number Of Partitions=$Emitter.PortCount)
   &SkeletalMeshLocation(
    Bone / Socket Index=$Particles.PartitionParticles.Partition
    Bone / Socket Sampling Mode=::NewEnumerator5
    MaskSpawnGroup=false
    Mesh Sampling Type=::NewEnumerator6
    Skeletal Mesh=$System.SkeletalMesh
   )
   &AddVelocity(
    Cone Angle=20
    Cone Axis=&MakeCustomVectorFromBool(
     Bool=true
     True Vector=&RandomRangeVector(Minimum=<1 0 -0.1>, Maximum=<1 0 0>)
     False Vector=&RandomRangeVector(Minimum=<1 0 -0.1>, Maximum=<1 0 -0.3>)
    )
    Inner Cone Angle=1
    Module.UseConeFalloff=false
    Module.UseVelDistribution=false
    Override Seed=false
    Velocity Mode=::NewEnumerator2
    Velocity Speed=&MakeCustomFloatFromBool(
     Bool=$User.bIsFrontendPreview
     True Float=&RandomRangeFloat(Minimum=-1500, Maximum=-2000)
     False Float=&RandomRangeFloat(Minimum=-2000, Maximum=-3000)
    )
   )
   &DLR_ProductPreviews(GameModeID:NiagaraInt32=$System.GameModeID, bPreviewVehicleSpeed:NiagaraBool=$System.bPreviewVehicleSpeed)
   &ShapeLocation(
    MaskSpawnGroup=false
    Override Seed=false
    Sphere Radius=&MakeCustomFloatFromBool(Bool=$User.bIsFrontendPreview, True Float=1, False Float=40)
    UseOwnerScale=true
   )
  ]
  particleUpdateScript [
   &ParticleState()
   /SetVariables(Particles.SparkSize:Vector2f=&Multiply_Vector2D_ByFloat(
     Float=&Multiply_Float(
      A=&RandomRangeFloat(Minimum=25, Maximum=55)
      B=&FloatFromCurve(FloatCurve=DataInterface() {propertyValues {Curve {
          defaultValue=1
          keys [
           {arriveTangent=4.4161, interpMode=RCIM_Cubic, leaveTangent=4.4161, tangentMode=RCTM_User, time=0, value=0}
           {arriveTangent=0.0439, interpMode=RCIM_Cubic, leaveTangent=0.0439, tangentMode=RCTM_User, time=0.3, value=1}
           {interpMode=RCIM_Cubic, tangentMode=RCTM_User, time=0.6, value=1}
           {tangentMode=RCTM_User, time=1, value=0}
          ]
         }}})
     )
     Vector2D=<1 1>
    )) {isSetParameters=true}
   /SetVariables() {isSetParameters=true}
   &ScaleSpriteSize(Uniform Curve Sprite Scale=DataInterface() {propertyValues {Curve {
       defaultValue=1
       keys [
        {arriveTangent=4.4161, interpMode=RCIM_Cubic, leaveTangent=4.4161, tangentMode=RCTM_User, time=0, value=0}
        {arriveTangent=0.0439, interpMode=RCIM_Cubic, leaveTangent=0.0439, tangentMode=RCTM_User, time=0.3, value=1}
        {interpMode=RCIM_Cubic, tangentMode=RCTM_User, time=0.5, value=1}
        {tangentMode=RCTM_User, time=1, value=0}
       ]
      }}})
   &InheritSourceMovement(
    Applied Position Scale=&VectorFromFloat(Value=&FloatFromCurve(FloatCurve=DataInterface() {propertyValues {Curve {
         defaultValue=1
         keys [
          {arriveTangent=-1.7885, interpMode=RCIM_Cubic, leaveTangent=-1.7885, tangentMode=RCTM_User, time=0.05, value=1}
          {interpMode=RCIM_Cubic, tangentMode=RCTM_User, time=1, value=0}
         ]
        }}}))
    ApplyForce=false
    ApplyPosition=true
    ApplyPositionFalloff=false
   )
   &Drag(Use Linear Drag=true, Use Rotational Drag=true)
   &SolveForcesAndVelocity(Clamp Velocity=false, Limit Acceleration=false)
   &DynamicMaterialParameters(
    Index 0 Param 1=1
    Index 0 Param 2=&MakeCustomFloatFromBool(Bool=$User.bIsFrontendPreview, True Float=3, False Float=3)
    Index 0 Param 3=1
    Module.Index0Float0Write=true
    Module.Index0Float1Write=true
    Module.Index0Float2Write=true
    Module.Index0Float3Write=true
   )
   &SubUVAnimation(
    "Sprite Renderer "=DataInterface() {propertyValues {SpriteRenderer=&self:NE_Champion_Star_0.NiagaraSpriteRendererProperties_0}}
    UseEndFrame=false
    UseStartFrame=false
   )
  ]
  rendererClasses [&NiagaraSpriteRendererProperties]
  renderers [
   {rendererClass=&NiagaraSpriteRendererProperties}
   {rendererClass=&NiagaraSpriteRendererProperties, rendererIndex=1}
  ]
 }
}
summary {
 systemName=NS_Champion
 emitters [
  {emitterName=NE_Champion_Sprite, rendererClasses [&NiagaraSpriteRendererProperties]}
  {emitterName=NE_Champion_Star, rendererClasses [&NiagaraSpriteRendererProperties]}
  {emitterName=NE_Champion_Ribbon, rendererClasses [&NiagaraRibbonRendererProperties]}
 ]
 userVariables [
  User.bIsFrontendPreview {type=NiagaraBool}
  User.bIsPainted {type=NiagaraBool}
  User.blsIsCoroColor {type=NiagaraBool}
  User.bPreviewVehicleSpeed {type=NiagaraBool}
  User.ColorVariantID {type=NiagaraInt32}
  User.GameModeID {type=NiagaraInt32}
  User.PaintedColor {type=LinearColor, description="", defaultValue {struct=LinearColor, value=<rgba 1 1 1 1>}}
  User.PaintedScalar {type=NiagaraFloat, defaultValue=0, description=""}
  User.PortCount {type=NiagaraInt32, defaultValue=2}
 ]
}
